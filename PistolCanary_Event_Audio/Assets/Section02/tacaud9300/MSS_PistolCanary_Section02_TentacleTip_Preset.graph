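# MetaSound Source preset for a Section 02 tentacle tip one-shot sound in the PistolCanary event.
# It inherits its graph from the stereo tentacle one-shots parent, exposing Wave Asset and Master Volume inputs, though the default page here leaves the wave unset.
# Playback is stereo with an OnFinished trigger, routed to the gameplay events sound class with large point-of-interest attenuation.
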MetaSoundSource /PistolCanary_Event_Audio/Assets/Section02/tacaud9300/MSS_PistolCanary_Section02_TentacleTip_Preset
preset_parent=/PistolCanary_Audio/Assets/MSS_PistolCanary_TentacleOneShots_Stereo
inputs [
 Master Volume {type=Float}
 UE.Source.OnPlay {type=Trigger}
 UE.Source.StartTime.StartTime {type=Time}
 Wave Asset {type=WaveAsset}
]
interfaces [UE.OutputFormat.Stereo:1.0, UE.Source.OneShot:1.0, UE.Source.StartTime:1.0, UE.Source:1.1]
outputs [
 "UE.OutputFormat.Stereo.Audio:0" {type=Audio}
 "UE.OutputFormat.Stereo.Audio:1" {type=Audio}
 UE.Source.OneShot.OnFinished {type=Trigger}
]
pages [Default {input_defaults {Master Volume=1, UE.Source.OnPlay=null, UE.Source.StartTime.StartTime=0, Wave Asset=null}}]
properties {
 assetImportData=&self:AssetImportData
 attenuationSettings=/PistolCanary_Audio/Assets/Settings/ATT_PistolCanary_POI_Large
 maxDistance=40000
 numChannels=2
 outputFormat=Stereo
 sampleRate=48000
 soundClassObject=&Fort_Events_Gameplay
}
